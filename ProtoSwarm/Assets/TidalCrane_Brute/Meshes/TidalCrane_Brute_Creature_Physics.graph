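# Physics asset for the TidalCrane Brute creature mesh, defining its collision and physics bodies.
# Capsules cover the head, pod base, three-segment front, middle and back legs on both sides, plus a three-capsule pod center.
# Constraints join segments within each leg and pod center to pod base, leaving the head and legs unconstrained to the pod.
# All swing and twist motions are Free, so the stored angle limits do not restrict rotation.
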PhysicsAsset /ProtoSwarm/Assets/TidalCrane_Brute/Meshes/TidalCrane_Brute_Creature_Physics
bodies {
 head [{shapeType=Capsule, length=30, radius=50.8646, center=<57.0153 0.7029 -19.2834>, rotation=<rot -69.8421 173.4276 -172.5164>}]
 leg_bk_01_l [{
   shapeType=Capsule
   length=43.8047
   radius=16.2493
   shapeName=leg_bk_01_l_capsule
   center=<27.6397 0.6135 1.2159>
   rotation=<rot 0 -91.0279 -80.2804>
  }]
 leg_bk_01_r [{
   shapeType=Capsule
   length=43.8047
   radius=16.2493
   shapeName=leg_bk_01_r_capsule
   center=<-27.6397 -0.6135 -1.2159>
   rotation=<rot 0 -91.0279 99.7196>
  }]
 leg_bk_02_l [{
   shapeType=Capsule
   length=38.7827
   radius=14.707
   shapeName=leg_bk_02_l_capsule
   center=<21.8643 1.2437 -2.4774>
   rotation=<rot 0.0001 93.4142 -78.2385>
  }]
 leg_bk_02_r [{
   shapeType=Capsule
   length=38.7827
   radius=14.707
   shapeName=leg_bk_02_r_capsule
   center=<-21.8643 -1.2437 2.4774>
   rotation=<rot 0.0001 93.4142 101.7615>
  }]
 leg_bk_03_l [{
   shapeType=Capsule
   length=71.4087
   radius=30.1258
   shapeName=leg_bk_03_l_capsule
   center=<50.3386 1.6352 23.3328>
   rotation=<rot -0.0001 -91.3073 -86.0828>
  }]
 leg_bk_03_r [{
   shapeType=Capsule
   length=71.4087
   radius=30.1258
   shapeName=leg_bk_03_r_capsule
   center=<-50.3386 -1.6352 -23.3328>
   rotation=<rot -0.0001 -91.3073 93.9172>
  }]
 leg_fr_01_l [{
   shapeType=Capsule
   length=39.7056
   radius=13.7787
   shapeName=leg_fr_01_l_capsule
   center=<32.979 -0.371 -0.3414>
   rotation=<rot 0 -89.9287 -87.6324>
  }]
 leg_fr_01_r [{
   shapeType=Capsule
   length=39.7056
   radius=13.7787
   shapeName=leg_fr_01_r_capsule
   center=<-32.979 0.371 0.3414>
   rotation=<rot 0 -89.9287 92.3676>
  }]
 leg_fr_02_l [{
   shapeType=Capsule
   length=72.6623
   radius=14.6042
   shapeName=leg_fr_02_l_capsule
   center=<34.1032 -0.4773 -0.4612>
   rotation=<rot 0 88.4561 -86.958>
  }]
 leg_fr_02_r [{
   shapeType=Capsule
   length=72.6623
   radius=14.6042
   shapeName=leg_fr_02_r_capsule
   center=<-34.1032 0.4773 0.4612>
   rotation=<rot 0 88.4561 93.042>
  }]
 leg_fr_03_l [{
   shapeType=Capsule
   length=105.9123
   radius=18.9205
   shapeName=leg_fr_03_l_capsule
   center=<60.0203 -3.8479 5.8304>
   rotation=<rot 0 90.2705 -87.8865>
  }]
 leg_fr_03_r [{
   shapeType=Capsule
   length=105.9123
   radius=18.9205
   shapeName=leg_fr_03_r_capsule
   center=<-60.0203 3.8479 -5.8304>
   rotation=<rot 0 90.2705 92.1135>
  }]
 leg_mid_01_l [{
   shapeType=Capsule
   length=35.9361
   radius=12.4249
   shapeName=leg_mid_01_l_capsule
   center=<26.5077 0.4872 1.2742>
   rotation=<rot 0 -88.6817 -80.4341>
  }]
 leg_mid_01_r [{
   shapeType=Capsule
   length=35.9361
   radius=12.4249
   shapeName=leg_mid_01_r_capsule
   center=<-26.5077 -0.4872 -1.2742>
   rotation=<rot 0 -88.6817 99.5659>
  }]
 leg_mid_02_l [{
   shapeType=Capsule
   length=48.9852
   radius=12.4229
   shapeName=leg_mid_02_l_capsule
   center=<32.8658 -0.3421 -0.0658>
   rotation=<rot 0 89.1321 -82.7364>
  }]
 leg_mid_02_r [{
   shapeType=Capsule
   length=48.9852
   radius=12.4229
   shapeName=leg_mid_02_r_capsule
   center=<-32.8658 0.3421 0.0658>
   rotation=<rot 0 89.1321 97.2636>
  }]
 leg_mid_03_l [{
   shapeType=Capsule
   length=61.9924
   radius=23.2584
   shapeName=leg_mid_03_l_capsule
   center=<34.7342 0.4782 13.9787>
   rotation=<rot 0 -90.7489 -89.0744>
  }]
 leg_mid_03_r [{
   shapeType=Capsule
   length=61.9924
   radius=23.2584
   shapeName=leg_mid_03_r_capsule
   center=<-34.7342 -0.4782 -13.9787>
   rotation=<rot 0 -90.7489 90.9256>
  }]
 pod_base [{shapeType=Capsule, length=45, radius=73, center=<73.0009 0.2721 50.9346>, rotation=<rot 0.603 -90 80.0005>}]
 pod_center [
  {shapeType=Capsule, length=88, radius=35, shapeName=pod_center_capsule, center=<11.8839 53.4931 -27.024>, rotation=<rot -90 -20 0>}
  {shapeType=Capsule, length=88, radius=35, shapeName=pod_center_capsule, center=<13.4314 -53.8957 -25.9712>, rotation=<rot -90 20 0>}
  {shapeType=Capsule, length=28.5, radius=33, shapeName=pod_center_capsule, center=<-95.6765 0 -59.3476>, rotation=<rot -0.0002 0 90>}
 ]
}
constraints [
 {
  bone1Name=pod_center
  bone2Name=pod_base
  swing1LimitDegrees=45
  swing1Motion=Free
  swing2LimitDegrees=45
  swing2Motion=Free
  twistLimitDegrees=45
  twistMotion=Free
 }
 {
  bone1Name=leg_bk_03_l
  bone2Name=leg_bk_02_l
  swing1LimitDegrees=45
  swing1Motion=Free
  swing2LimitDegrees=45
  swing2Motion=Free
  twistLimitDegrees=45
  twistMotion=Free
 }
 {
  bone1Name=leg_bk_02_l
  bone2Name=leg_bk_01_l
  swing1LimitDegrees=45
  swing1Motion=Free
  swing2LimitDegrees=45
  swing2Motion=Free
  twistLimitDegrees=45
  twistMotion=Free
 }
 {
  bone1Name=leg_mid_03_l
  bone2Name=leg_mid_02_l
  swing1LimitDegrees=45
  swing1Motion=Free
  swing2LimitDegrees=45
  swing2Motion=Free
  twistLimitDegrees=45
  twistMotion=Free
 }
 {
  bone1Name=leg_mid_02_l
  bone2Name=leg_mid_01_l
  swing1LimitDegrees=45
  swing1Motion=Free
  swing2LimitDegrees=45
  swing2Motion=Free
  twistLimitDegrees=45
  twistMotion=Free
 }
 {
  bone1Name=leg_fr_03_l
  bone2Name=leg_fr_02_l
  swing1LimitDegrees=45
  swing1Motion=Free
  swing2LimitDegrees=45
  swing2Motion=Free
  twistLimitDegrees=45
  twistMotion=Free
 }
 {
  bone1Name=leg_fr_02_l
  bone2Name=leg_fr_01_l
  swing1LimitDegrees=45
  swing1Motion=Free
  swing2LimitDegrees=45
  swing2Motion=Free
  twistLimitDegrees=45
  twistMotion=Free
 }
 {
  bone1Name=leg_fr_02_r
  bone2Name=leg_fr_01_r
  swing1LimitDegrees=45
  swing1Motion=Free
  swing2LimitDegrees=45
  swing2Motion=Free
  twistLimitDegrees=45
  twistMotion=Free
 }
 {
  bone1Name=leg_fr_03_r
  bone2Name=leg_fr_02_r
  swing1LimitDegrees=45
  swing1Motion=Free
  swing2LimitDegrees=45
  swing2Motion=Free
  twistLimitDegrees=45
  twistMotion=Free
 }
 {
  bone1Name=leg_mid_02_r
  bone2Name=leg_mid_01_r
  swing1LimitDegrees=45
  swing1Motion=Free
  swing2LimitDegrees=45
  swing2Motion=Free
  twistLimitDegrees=45
  twistMotion=Free
 }
 {
  bone1Name=leg_mid_03_r
  bone2Name=leg_mid_02_r
  swing1LimitDegrees=45
  swing1Motion=Free
  swing2LimitDegrees=45
  swing2Motion=Free
  twistLimitDegrees=45
  twistMotion=Free
 }
 {
  bone1Name=leg_bk_02_r
  bone2Name=leg_bk_01_r
  swing1LimitDegrees=45
  swing1Motion=Free
  swing2LimitDegrees=45
  swing2Motion=Free
  twistLimitDegrees=45
  twistMotion=Free
 }
 {
  bone1Name=leg_bk_03_r
  bone2Name=leg_bk_02_r
  swing1LimitDegrees=45
  swing1Motion=Free
  swing2LimitDegrees=45
  swing2Motion=Free
  twistLimitDegrees=45
  twistMotion=Free
 }
]
